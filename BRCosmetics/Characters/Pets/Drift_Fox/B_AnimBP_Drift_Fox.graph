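# Animation Blueprint for the Drift Fox pet cosmetic, targeting the shared Dog skeleton used by pets.
# The event graph is empty and no variables are declared, so the asset adds no event logic of its own.
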AnimBlueprint /BRCosmetics/Characters/Pets/Drift_Fox/B_AnimBP_Drift_Fox
target_skeleton=/Game/Characters/Pets/Dog_Shared/Dog_Skeleton
graphs [EventGraph {links [], nodes []}]
variables []
